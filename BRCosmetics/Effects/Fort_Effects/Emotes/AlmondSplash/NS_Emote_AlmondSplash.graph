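# Niagara system for the AlmondSplash emote that adds snow dust, flakes and a snow wake at bones or sockets on the character's skeletal mesh.
# SnowDust sprites scale velocity with system speed, while SnowFlakes spawn only during forward movement and drift on curl noise.
# SnowWake holds two persistent mirrored meshes, tinted by time of day, feeding smoothed world position and timeline progress into their material.
# Alternate dust and flake emitters are disabled.
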NiagaraSystem /BRCosmetics/Effects/Fort_Effects/Emotes/AlmondSplash/NS_Emote_AlmondSplash
dependencies {
 usedDataInterfaces [
  &NiagaraDataInterfaceCurve
  &NiagaraDataInterfaceVectorCurve
  &NiagaraDataInterfaceSpriteRendererInfo
  &NiagaraDataInterfaceArrayFloat3
  &NiagaraDataInterfaceArrayFloat
  &NiagaraDataInterfaceColorCurve
  &NiagaraDataInterfaceSkeletalMesh
 ]
 usedDynamicInputs [
  &TransformVector
  &Multiply_Vector
  &SetBoolByFloatComparison
 ]
 usedModules [
  &self:SystemSpawnScript.SystemScriptSource.SystemScriptGraph.NiagaraNodeAssignment_0.SetVariables_#1
  &SystemState
  &self:SystemSpawnScript.SystemScriptSource.SystemScriptGraph.NiagaraNodeAssignment_1.SetVariables_#2
  &self:SystemSpawnScript.SystemScriptSource.SystemScriptGraph.NiagaraNodeAssignment_2.SetVariables_#3
  &self:SystemSpawnScript.SystemScriptSource.SystemScriptGraph.NiagaraNodeAssignment_3.SetVariables_#4
  &self:SystemSpawnScript.SystemScriptSource.SystemScriptGraph.NiagaraNodeAssignment_4.SetVariables_#5
  &Timeline
 ]
}
emitters {
 SnowDust {
  emitterSpawnScript [/SetVariables() {isSetParameters=true, module=SetVariables_#6}]
  emitterUpdateScript [
   &EmitterState()
   &SpawnPerUnit(Module.Use Max Movement Threshold=true, Module.Use Movement Tolerance=false, Use Spawn Probability=false) {enabled=false}
   &SpawnRate(SpawnRate=10, Use Spawn Probability=false)
  ]
  particleSpawnScript [
   &InitializeParticle(
    Color Mode=::NewEnumerator0
    Lifetime Max=4
    Lifetime Min=3
    Lifetime Mode=::NewEnumerator1
    Sprite Rotation Mode=::NewEnumerator2
    Sprite Size Mode=::NewEnumerator1
    Uniform Sprite Size Max=150
    Uniform Sprite Size Min=120
    UsePositionOffset=true
    Write Material Random=true
   )
   &PartitionParticles()
   &SkeletalMeshLocation(
    Bone / Socket Index=$Particles.PartitionParticles.Partition
    MaskSpawnGroup=false
    Mesh Sampling Type=::NewEnumerator5
    Sampled Position Offset=<0 0 -10>
    Skeletal Mesh=$System.Skeletal Mesh
    Socket Sampling Mode=::NewEnumerator5
   )
   &ShapeLocation(MaskSpawnGroup=false, Sphere Distribution=::NewEnumerator3, Sphere Radius=1, UseOwnerScale=true)
   &AddVelocity(
    Cone Angle=20
    Cone Axis=<0 -1 0>
    Module.UseConeFalloff=false
    Module.UseVelDistribution=false
    Override Seed=false
    Velocity Mode=::NewEnumerator2
    Velocity Speed=&RemapRange(Input Value To Remap=$System.Speed, Input Min=0, Input Max=130, Output Min=50, Output Max=60)
   )
   &CameraOffset(Camera Offset Amount=10)
  ]
  particleUpdateScript [
   &ParticleState()
   /SetVariables(
    Particles.Dust Color:LinearColor=&Multiply_LinearColorByFloat(
     Linear Color=<rgba 1 1 1 1>
     Float=&FloatFromCurve(
      FloatCurve=DataInterface() {propertyValues {Curve {keys [
          {interpMode=RCIM_Cubic, tangentMode=RCTM_User, time=0, value=0}
          {arriveTangent=5.671, interpMode=RCIM_Cubic, leaveTangent=5.671, tangentMode=RCTM_User, time=0.2, value=0.35}
          {interpMode=RCIM_Cubic, leaveTangent=0.0874, tangentMode=RCTM_Break, time=0.3, value=1}
          {arriveTangent=-2, interpMode=RCIM_Cubic, leaveTangent=-2, time=0.5, value=0.35}
          {tangentMode=RCTM_User, time=0.8, value=0}
         ]}}}
      Scale Curve=&RandomRangeFloat(Minimum=0.5, Maximum=1)
     )
    )
    Particles.Dust Dyn Mat Par:Vector4f=&MakeVector4(X=3, Y=1, Z=0, W=1)
   ) {isSetParameters=true, module=SetVariables_#7}
   /SetVariables(
    Particles.Clouds Color:LinearColor=&Multiply_LinearColorByFloat(
     Linear Color=<rgba 0.6995 0.8347 1 0.3>
     Float=&FloatFromCurve(
      FloatCurve=DataInterface() {propertyValues {Curve {keys [
          {arriveTangent=9.0725, interpMode=RCIM_Cubic, leaveTangent=9.0725, tangentMode=RCTM_User, time=0, value=0}
          {arriveTangent=0.1145, interpMode=RCIM_Cubic, leaveTangent=0.1145, tangentMode=RCTM_User, time=0.2, value=0.95}
          {arriveTangent=0.1709, interpMode=RCIM_Cubic, leaveTangent=0.1709, tangentMode=RCTM_User, time=0.4, value=0.95}
          {arriveTangent=-0.3021, interpMode=RCIM_Cubic, leaveTangent=-0.3021, tangentMode=RCTM_User, time=1, value=0}
         ]}}}
      Scale Curve=&Multiply_Float(
       A=&RandomRangeFloat(Minimum=0.5, Maximum=1)
       B=&Lerp_Float(A=0.5, B=1, Alpha=$System.Timeline.Time)
      )
     )
    )
    Particles.Clouds Dyn Mat Par:Vector4f=&MakeVector4(X=2, Y=2, Z=2, W=0)
   ) {isSetParameters=true, module=SetVariables_#8}
   &SpriteRotationRate(Rotation Rate=&RandomRangeFloat(Minimum=10, Maximum=40))
   &GravityForce(Gravity=&Multiply_VectorByFloat(
     Vector=<1 1 70>
     Float=&FloatFromCurve(FloatCurve=DataInterface() {propertyValues {Curve {keys [<at 0 0 RCIM_Cubic>, {arriveTangent=2.3028, interpMode=RCIM_Cubic, leaveTangent=2.3028, tangentMode=RCTM_User, time=1, value=1}]}}})
    ))
   &InheritSourceMovement(
    Applied Position Falloff Distance=5
    Applied Position Scale=&VectorFromCurve(
     VectorCurve=DataInterface() {propertyValues {XCurve {keys [<at 0 1 RCIM_Cubic>,<at 1 0.8>]}, YCurve {keys [<at 0 0>,<at 1 0>]}, ZCurve {keys [<at 0 1 RCIM_Cubic>,<at 1 0.8>]}}}
     Scale Curve=&VectorFromFloat(Value=0.1)
    )
    ApplyForce=false
    ApplyPosition=true
    ApplyPositionFalloff=false
   )
   &Drag(Use Linear Drag=true, Use Rotational Drag=true) {enabled=false}
   /Niagara/Modules/Update/Utility/IncrementOverTime(Rate of Change:NiagaraFloat=&Multiply_Float(A=&VectorLength(Vector=$Particles.Velocity), B=0.005))
   &SolveForcesAndVelocity(Clamp Velocity=false, Limit Acceleration=false)
   &SubUVAnimation(
    "Sprite Renderer "=DataInterface() {propertyValues {SpriteRenderer=&self:SnowDust_1.NiagaraSpriteRendererProperties_1}}
    UseEndFrame=true
    UseStartFrame=true
   )
   &ScaleSpriteSize(
    Uniform Curve Scale=0.5
    Uniform Curve Sprite Scale=DataInterface() {propertyValues {Curve {keys [
        {arriveTangent=2.2311, interpMode=RCIM_Cubic, leaveTangent=2.2311, tangentMode=RCTM_User, time=0, value=0.4}
        {interpMode=RCIM_Cubic, leaveTangent=-0.0405, time=1, value=1}
       ]}}}
   )
  ]
  rendererClasses [&NiagaraSpriteRendererProperties]
  renderers [
   {rendererClass=&NiagaraSpriteRendererProperties}
   {rendererClass=&NiagaraSpriteRendererProperties, rendererIndex=1}
  ]
 }
 SnowDust001 {
  bEnabled=false
  emitterSpawnScript [/SetVariables() {isSetParameters=true, module=SetVariables_#6}]
  emitterUpdateScript [
   &EmitterState()
   &SpawnPerUnit(Module.Use Max Movement Threshold=true, Module.Use Movement Tolerance=false, Use Spawn Probability=false) {enabled=false}
   &SpawnRate(SpawnRate=7, Use Spawn Probability=false)
  ]
  particleSpawnScript [
   &InitializeParticle(
    Color=<rgba 0.5493 0.7469 1 1>
    Lifetime Max=5
    Lifetime Min=4
    Lifetime Mode=::NewEnumerator1
    Sprite Rotation Mode=::NewEnumerator2
    Sprite Size Mode=::NewEnumerator1
    Uniform Sprite Size Max=150
    Uniform Sprite Size Min=120
    UsePositionOffset=false
    Write Material Random=true
   )
   &PartitionParticles(Number Of Partitions=4)
   &SkeletalMeshLocation(
    Bone / Socket Index=$Particles.PartitionParticles.Partition
    MaskSpawnGroup=false
    Mesh Sampling Type=::NewEnumerator5
    Skeletal Mesh=$System.Skeletal Mesh
    Socket Sampling Mode=::NewEnumerator5
   )
   &ShapeLocation(
    MaskSpawnGroup=false
    Non Uniform Scale=&SelectVectorFromArray(
     Direct Array Index=$Particles.PartitionParticles.Partition
     Vector Selection Array=DataInterface() {propertyValues {FloatData [<0.1 1 0.1>, <0.1 1 0.1>, <0.1 0.25 0.1>, <0.1 0.25 0.1>]}}
    )
    Sphere Distribution=::NewEnumerator3
    Sphere Radius=80
    UseOwnerScale=true
   )
   &InheritVelocity(
    Inherited Velocity Amount Scale=&VectorFromFloat(Value=&FloatFromCurve(
      FloatCurve=DataInterface() {propertyValues {Curve {keys [{interpMode=RCIM_Cubic, tangentMode=RCTM_User, time=0, value=0}, {interpMode=RCIM_Cubic, tangentMode=RCTM_User, time=0.5, value=1}]}}}
      CurveIndex=$Emitter.Age
      Scale Curve=0.5
     ))
    Inherited Velocity Speed Limit=125
    Speed Limit=true
    UseSpeedThreshold=false
   ) {enabled=false}
   &AddVelocity(
    Velocity=<0 -1 0>
    Velocity Speed Scale=&RemapRange(Input Value To Remap=$System.Speed, Input Min=0, Input Max=105, Output Min=100, Output Max=0)
   ) {enabled=false}
   &AddVelocity(Velocity=<0 -1 0>, Velocity Speed Scale=100) {module=AddVelocity001}
  ]
  particleUpdateScript [
   &ParticleState()
   /SetVariables(
    Particles.Dust Color:LinearColor=&Multiply_LinearColorByFloat(
     Linear Color=<rgba 1 1 1 0.2>
     Float=&FloatFromCurve(
      FloatCurve=DataInterface() {propertyValues {Curve {keys [
          {interpMode=RCIM_Cubic, tangentMode=RCTM_User, time=0, value=0}
          {arriveTangent=5.671, interpMode=RCIM_Cubic, leaveTangent=5.671, tangentMode=RCTM_User, time=0.2, value=0.35}
          {interpMode=RCIM_Cubic, leaveTangent=0.0874, tangentMode=RCTM_Break, time=0.3, value=1}
          {arriveTangent=-2, interpMode=RCIM_Cubic, leaveTangent=-2, time=0.5, value=0.35}
          {tangentMode=RCTM_User, time=0.8, value=0}
         ]}}}
      Scale Curve=&RandomRangeFloat(Minimum=0.5, Maximum=1)
     )
    )
    Particles.Dust Dyn Mat Par:Vector4f=&MakeVector4(X=3, Y=1, Z=0, W=1)
   ) {isSetParameters=true, module=SetVariables_#7}
   /SetVariables(
    Particles.Clouds Color:LinearColor=&Multiply_LinearColorByFloat(
     Linear Color=&RandomRangeLinearColor(Minimum=<rgba 0.4948 0.7163 1 0.25>, Maximum=<rgba 0.349 0.6344 1 0.25>)
     Float=&FloatFromCurve(
      FloatCurve=DataInterface() {propertyValues {Curve {keys [
          {arriveTangent=9.0725, interpMode=RCIM_Cubic, leaveTangent=9.0725, tangentMode=RCTM_User, time=0, value=0}
          {arriveTangent=-2.5249, interpMode=RCIM_Cubic, leaveTangent=-2.5249, tangentMode=RCTM_User, time=0.1408, value=0.9847}
          {arriveTangent=-0.0405, interpMode=RCIM_Cubic, leaveTangent=-0.0405, tangentMode=RCTM_User, time=1, value=0}
         ]}}}
      Scale Curve=&RandomRangeFloat(Minimum=0.5, Maximum=1)
     )
    )
    Particles.Clouds Dyn Mat Par:Vector4f=&MakeVector4(X=2, Y=0.5, Z=2, W=0)
   ) {isSetParameters=true, module=SetVariables_#8}
   &SpriteRotationRate(Rotation Rate=&RandomRangeFloat(Minimum=10, Maximum=40))
   &Drag(Drag=&RandomRangeFloat(Minimum=1, Maximum=1.25), Use Linear Drag=true, Use Rotational Drag=true)
   &SolveForcesAndVelocity(Clamp Velocity=false, Limit Acceleration=false)
   &SubUVAnimation(
    "Sprite Renderer "=DataInterface() {propertyValues {SpriteRenderer=&self:SnowDust001_1.NiagaraSpriteRendererProperties_1}}
    UseEndFrame=true
    UseStartFrame=true
   )
   &ScaleSpriteSize(Uniform Curve Sprite Scale=DataInterface() {propertyValues {Curve {keys [
        {arriveTangent=2.2461, interpMode=RCIM_Cubic, leaveTangent=2.2461, tangentMode=RCTM_User, time=0, value=0.4}
        {interpMode=RCIM_Cubic, leaveTangent=-0.0405, time=1, value=1}
       ]}}})
   &CameraOffset(Camera Offset Amount=&FloatFromCurve(
     FloatCurve=DataInterface() {propertyValues {Curve {keys [{interpMode=RCIM_Cubic, tangentMode=RCTM_User, time=0, value=0}, {interpMode=RCIM_Cubic, tangentMode=RCTM_User, time=1, value=1}]}}}
     Scale Curve=10
    ))
  ]
  rendererClasses [&NiagaraSpriteRendererProperties]
  renderers [
   {rendererClass=&NiagaraSpriteRendererProperties}
   {rendererClass=&NiagaraSpriteRendererProperties, rendererIndex=1}
  ]
 }
 SnowFlakes {
  emitterUpdateScript [
   &EmitterState()
   &SpawnRate(
    SpawnRate=&MakeCustomFloatFromBool(Bool=$System.isMovingForward, True Float=20, False Float=0)
    Use Spawn Probability=false
   )
  ]
  particleSpawnScript [
   &InitializeParticle(
    Color=<rgba 1.9412 2.8115 5 1>
    Lifetime Max=1.2
    Lifetime Min=0.7
    Lifetime Mode=::NewEnumerator1
    Sprite Rotation Mode=::NewEnumerator2
    Sprite Size Mode=::NewEnumerator1
    Uniform Sprite Size Max=15
    UsePositionOffset=false
    Write Material Random=true
   )
   &PartitionParticles()
   &SkeletalMeshLocation(
    Bone / Socket Index=&Add_Integer(A=$Particles.PartitionParticles.Partition, B=2)
    MaskSpawnGroup=false
    Mesh Sampling Type=::NewEnumerator5
    Skeletal Mesh=$System.Skeletal Mesh
    Socket Sampling Mode=::NewEnumerator5
   )
   &ShapeLocation(MaskSpawnGroup=false, Non Uniform Scale=<10 60 10>, Override Seed=false, Sphere Radius=1, UseOwnerScale=true)
   &AddVelocity(
    Cone Angle=30
    Cone Axis=<0 -1 0>
    Module.UseConeFalloff=false
    Module.UseVelDistribution=false
    Override Seed=false
    Velocity Mode=::NewEnumerator2
    Velocity Speed=60
   )
   /SetVariables(Particles.VisibilityTag=&RandomRangeInt(Minimum=0, Maximum=1)) {isSetParameters=true, module=SetVariables_#9}
   &DynamicMaterialParameters(
    Index 0 Param 1=0.5
    Index 0 Param 2=&FloatFromCurve(FloatCurve=DataInterface() {propertyValues {Curve {keys [{arriveTangent=-12.4095, interpMode=RCIM_Cubic, leaveTangent=-12.4095, tangentMode=RCTM_User, time=0.5, value=2}, <at 1 0 RCIM_Cubic>]}}})
    Module.Param0WriteEnabled=true
    Module.Param1WriteEnabled=true
    Module.Param2WriteEnabled=true
    Module.Param3WriteEnabled=true
   ) {module=DynamicMaterialParameters001}
   &CameraOffset(Camera Offset Amount=50)
  ]
  particleUpdateScript [
   &ParticleState()
   &SpriteRotationRate(Rotation Rate=&RandomRangeFloat(Minimum=-1, Maximum=1))
   &Drag(Drag=&RandomRangeFloat(Minimum=1, Maximum=1.25), Use Linear Drag=true, Use Rotational Drag=true) {enabled=false}
   &GravityForce(Gravity=<0 0 100>)
   &CurlNoiseForce(Bias Noise=false, Noise Strength=100, Pan Noise=false)
   &SolveForcesAndVelocity(Clamp Velocity=false, Limit Acceleration=false)
   &ScaleSpriteSize(Uniform Curve Sprite Scale=DataInterface() {propertyValues {Curve {keys [<at 0 0 RCIM_Cubic>,<at 0.1351 0.878 RCIM_Cubic>,<at 1 0 RCIM_Cubic>]}}})
   /SetVariables(Particles.Sprite02_Size:Vector2f=&Vector2DFromFloat(Value=&Multiply_Float(
      A=&RandomRangeFloat(Minimum=10, Maximum=20)
      B=&FloatFromCurve(FloatCurve=DataInterface() {propertyValues {Curve {keys [<at 0 0 RCIM_Cubic>,<at 0.1351 0.878 RCIM_Cubic>,<at 1 0 RCIM_Cubic>]}}})
     ))) {isSetParameters=true, module=SetVariables_#10}
   &ScaleColor(
    Scale Alpha=&Multiply_Float(
     A=&RandomRangeFloat(Minimum=0.2, Maximum=1)
     B=&FloatFromCurve(
      FloatCurve=DataInterface() {propertyValues {Curve {keys [
          {arriveTangent=4.1785, interpMode=RCIM_Cubic, leaveTangent=4.1785, tangentMode=RCTM_User, time=0, value=0}
          {arriveTangent=0.9995, interpMode=RCIM_Cubic, leaveTangent=0.9995, tangentMode=RCTM_User, time=0.2223, value=0.6825}
          {arriveTangent=-0.0405, interpMode=RCIM_Cubic, leaveTangent=-0.0405, tangentMode=RCTM_User, time=1, value=1}
         ]}}}
      Scale Curve=&FNTimeOfDayScalar(Night Scale=1, Morning Scale=0.2, Evening Scale=0.4)
     )
    )
    ScaleA=true
    ScaleRGB=true
   )
  ]
  rendererClasses [&NiagaraSpriteRendererProperties]
  renderers [
   {rendererClass=&NiagaraSpriteRendererProperties}
   {rendererClass=&NiagaraSpriteRendererProperties, rendererIndex=1}
  ]
 }
 SnowFlakes001 {
  bEnabled=false
  emitterUpdateScript [
   &EmitterState()
   &SpawnRate(SpawnRate=20, Use Spawn Probability=false)
  ]
  particleSpawnScript [
   &InitializeParticle(
    Color=<rgba 1.9412 2.8115 5 1>
    Lifetime Max=1.2
    Lifetime Min=0.7
    Lifetime Mode=::NewEnumerator1
    Sprite Rotation Mode=::NewEnumerator2
    Sprite Size Mode=::NewEnumerator1
    Uniform Sprite Size Max=15
    UsePositionOffset=false
    Write Material Random=true
   )
   &PartitionParticles(Number Of Partitions=4)
   &SkeletalMeshLocation(
    Bone / Socket Index=$Particles.PartitionParticles.Partition
    MaskSpawnGroup=false
    Mesh Sampling Type=::NewEnumerator5
    Skeletal Mesh=$System.Skeletal Mesh
    Socket Sampling Mode=::NewEnumerator5
   )
   &ShapeLocation(
    MaskSpawnGroup=false
    Non Uniform Scale=&SelectVectorFromArray(
     Direct Array Index=$Particles.PartitionParticles.Partition
     Vector Selection Array=DataInterface() {propertyValues {FloatData [<0.1 1 0.1>, <0.1 1 0.1>, <0.1 0.25 0.1>, <0.1 0.25 0.1>]}}
    )
    Override Seed=false
    UseOwnerScale=true
   )
   &InheritVelocity(Inherited Velocity Speed Limit=125, Speed Limit=true, UseSpeedThreshold=false)
   &AddVelocity(
    Cone Angle=90
    Cone Axis=<0 -1 0>
    Module.UseConeFalloff=false
    Module.UseVelDistribution=false
    Override Seed=false
    Velocity Mode=::NewEnumerator2
    Velocity Speed=&RemapRange(
     Input Value To Remap=$System.Speed
     Input Min=0
     Input Max=105
     Output Min=&RandomRangeFloat(Minimum=100, Maximum=120)
     Output Max=0
    )
   )
   /SetVariables(Particles.VisibilityTag=&RandomRangeInt(Minimum=0, Maximum=1)) {isSetParameters=true, module=SetVariables_#9}
   &DynamicMaterialParameters(
    Index 0 Param 1=0.5
    Index 0 Param 2=&FloatFromCurve(FloatCurve=DataInterface() {propertyValues {Curve {keys [{arriveTangent=-12.4095, interpMode=RCIM_Cubic, leaveTangent=-12.4095, tangentMode=RCTM_User, time=0.5, value=2}, <at 1 0 RCIM_Cubic>]}}})
    Module.Param0WriteEnabled=true
    Module.Param1WriteEnabled=true
    Module.Param2WriteEnabled=true
    Module.Param3WriteEnabled=true
   ) {module=DynamicMaterialParameters001}
  ]
  particleUpdateScript [
   &ParticleState()
   &SpriteRotationRate(Rotation Rate=&RandomRangeFloat(Minimum=-1, Maximum=1))
   &Drag(Drag=&RandomRangeFloat(Minimum=1, Maximum=1.25), Use Linear Drag=true, Use Rotational Drag=true)
   &GravityForce(Gravity=<0 0 100>)
   &CurlNoiseForce(Bias Noise=false, Noise Strength=100, Pan Noise=false)
   &SolveForcesAndVelocity(Clamp Velocity=false, Limit Acceleration=false)
   &ScaleSpriteSize(Uniform Curve Sprite Scale=DataInterface() {propertyValues {Curve {keys [<at 0 0 RCIM_Cubic>,<at 0.1351 0.878 RCIM_Cubic>,<at 1 0 RCIM_Cubic>]}}})
   /SetVariables(Particles.Sprite02_Size:Vector2f=&Vector2DFromFloat(Value=&Multiply_Float(
      A=&RandomRangeFloat(Minimum=10, Maximum=20)
      B=&FloatFromCurve(FloatCurve=DataInterface() {propertyValues {Curve {keys [<at 0 0 RCIM_Cubic>,<at 0.1351 0.878 RCIM_Cubic>,<at 1 0 RCIM_Cubic>]}}})
     ))) {isSetParameters=true, module=SetVariables_#10}
   &CameraOffset()
   &ScaleColor(
    Scale Alpha=&Multiply_Float(
     A=&RandomRangeFloat(Minimum=0.2, Maximum=1)
     B=&FloatFromCurve(
      FloatCurve=DataInterface() {propertyValues {Curve {keys [
          {arriveTangent=4.1785, interpMode=RCIM_Cubic, leaveTangent=4.1785, tangentMode=RCTM_User, time=0, value=0}
          {arriveTangent=0.9995, interpMode=RCIM_Cubic, leaveTangent=0.9995, tangentMode=RCTM_User, time=0.2223, value=0.6825}
          {arriveTangent=-0.0405, interpMode=RCIM_Cubic, leaveTangent=-0.0405, tangentMode=RCTM_User, time=1, value=1}
         ]}}}
      Scale Curve=&FNTimeOfDayScalar(Night Scale=1, Morning Scale=0.2, Evening Scale=0.4)
     )
    )
    ScaleA=true
    ScaleRGB=true
   )
  ]
  rendererClasses [&NiagaraSpriteRendererProperties]
  renderers [
   {rendererClass=&NiagaraSpriteRendererProperties}
   {rendererClass=&NiagaraSpriteRendererProperties, rendererIndex=1}
  ]
 }
 SnowWake {
  emitterUpdateScript [
   &EmitterState(Inactive Response=::NewEnumerator1, Life Cycle Mode=::NewEnumerator1, Loop Duration=1, UseLoopDelay=false)
   &SpawnBurst_Instantaneous(Spawn Count=2, Use Spawn Probability=false, UseLoopCountLimit=true)
  ]
  particleSpawnScript [
   &PartitionParticles()
   &InitializeParticle(
    Lifetime=5
    Mesh Scale=&MakeVector(
     X=&SelectFloatFromArray(Direct Array Index=$Particles.PartitionParticles.Partition, Float Selection Array=DataInterface() {propertyValues {FloatData [1,-1]}})
     Y=0.8
     Z=1
    )
    Mesh Scale Mode=::NewEnumerator4
    UsePositionOffset=false
    Write Material Random=true
   )
   &InitialMeshOrientation(
    Mesh Axis To Orient=<0 -1 0>
    Mesh Orientation Mode=::NewEnumerator2
    Orientation Vector=<0 1 0>
    Rotation=&SelectVectorFromArray(
     Direct Array Index=$Particles.PartitionParticles.Partition
     Vector Selection Array=DataInterface() {propertyValues {FloatData [<0 0 -0.045>, <0 0 0.045>]}}
    )
    Use Rotation Vector=true
   )
   &SkeletalMeshLocation(
    Bone / Socket Index=$Particles.PartitionParticles.Partition
    MaskSpawnGroup=false
    Mesh Sampling Type=::NewEnumerator5
    Orientation Sampling=::NewEnumerator1
    Sampled Position Offset=&SelectVectorFromArray(
     Direct Array Index=$Particles.PartitionParticles.Partition
     Vector Selection Array=DataInterface() {propertyValues {FloatData [<-30 7 35>, <30 7 35>]}}
    )
    Skeletal Mesh=$System.Skeletal Mesh
    Socket Sampling Mode=::NewEnumerator5
   )
  ]
  particleUpdateScript [
   &ParticleState(Kill Particles When Lifetime Has Elapsed=false)
   &InterpolateOverTime(
    Rate Of Change=10
    Target Vector Value:Vector3f=&TransformPosition(Position=$Particles.Position, Source Space=::Local, Destination Space=::World)
    Variable Type=::NewEnumerator2
   )
   &DynamicMaterialParameters(
    Index 0 Param 1=&RandomRangeFloat(Minimum=0, Maximum=1)
    Index 0 Param 2=&Lerp_Float(A=1, B=0, Alpha=$System.Timeline.Time)
    Index 1 Param 1=&MakeFloatFromVector(Vector=$StackContext.InterpolateOverTime.InterpolateOverTime_V3_.Moving Average, Channel=::NewEnumerator0)
    Index 1 Param 2=&MakeFloatFromVector(Vector=$StackContext.InterpolateOverTime.InterpolateOverTime_V3_.Moving Average, Channel=::NewEnumerator1)
    Index 1 Param 3=&MakeFloatFromVector(Vector=$StackContext.InterpolateOverTime.InterpolateOverTime_V3_.Moving Average, Channel=::NewEnumerator2)
    Module.Index0Param0Write=true
    Module.Index0Param1Write=true
    Module.Index0Param2Write=true
    Module.Index0Param3Write=true
    Module.Index1Param0Write=true
    Module.Index1Param1Write=true
    Module.Index1Param2Write=true
    Module.Index1Param3Write=true
    Write Parameter Index 1=true
   )
   &ScaleColor(
    Scale Alpha=&FloatFromCurve(
     FloatCurve=DataInterface() {propertyValues {Curve {keys [
         {interpMode=RCIM_Cubic, tangentMode=RCTM_User, time=0, value=0.5}
         {interpMode=RCIM_Cubic, tangentMode=RCTM_User, time=0.5, value=1}
         {tangentMode=RCTM_User, time=1, value=0.5}
        ]}}}
     CurveIndex=Linked() {linkedVariable {name=NPC.FortniteNPC.TimeOfDay_0-1, type=NiagaraFloat}}
    )
    Scale RGB=&MakeVectorFromLinearColorRGB(Color=&ColorFromCurve(
      ColorCurve=DataInterface() {propertyValues {
        AlphaCurve {keys [<at 0 1>,<at 1 1>]}
        BlueCurve {keys [
          {arriveTangent=0.256, arriveTangentWeight=0.0864, interpMode=RCIM_Cubic, leaveTangentWeight=0.0864, time=0.001, value=0.2157}
          <at 0.1097 0.2157>
          {
           arriveTangent=-0.0692
           arriveTangentWeight=0.0231
           interpMode=RCIM_Cubic
           leaveTangent=-0.0692
           leaveTangentWeight=0.0231
           time=0.1401
           value=0.2452
          }
          {
           arriveTangent=-2.1444
           arriveTangentWeight=0.0556
           interpMode=RCIM_Cubic
           leaveTangent=-2.1444
           leaveTangentWeight=0.0556
           time=0.171
           value=0.2039
          }
          {arriveTangent=0.7909, interpMode=RCIM_Cubic, leaveTangent=0.7909, time=0.216, value=0.0824}
          {interpMode=RCIM_Cubic, tangentMode=RCTM_User, time=0.29, value=0.298}
          {
           arriveTangent=1.194
           arriveTangentWeight=0.0789
           interpMode=RCIM_Cubic
           leaveTangent=1.194
           leaveTangentWeight=0.0789
           tangentMode=RCTM_User
           tangentWeightMode=RCTWM_WeightedBoth
           time=0.375
           value=0.3229
          }
          {
           arriveTangent=0.246
           arriveTangentWeight=0.0709
           interpMode=RCIM_Cubic
           leaveTangent=0.246
           leaveTangentWeight=0.0709
           tangentMode=RCTM_User
           tangentWeightMode=RCTWM_WeightedBoth
           time=0.5
           value=0.7
          }
          {interpMode=RCIM_Cubic, tangentMode=RCTM_User, time=0.817, value=0.4588}
          {arriveTangent=-3.5889, interpMode=RCIM_Cubic, leaveTangent=-3.5889, time=0.833, value=0.4392}
          {
           arriveTangent=-2.455
           arriveTangentWeight=0.2296
           interpMode=RCIM_Cubic
           leaveTangent=-2.455
           leaveTangentWeight=0.2105
           time=0.898
           value=0.1681
          }
          {arriveTangent=0.4796, interpMode=RCIM_Cubic, leaveTangent=0.4796, time=0.9285, value=0.2047}
          {arriveTangentWeight=0.1844, interpMode=RCIM_Cubic, leaveTangent=0.0005, leaveTangentWeight=0.1844, time=1, value=0.217}
         ]}
        GreenCurve {keys [
          {arriveTangent=0.2383, arriveTangentWeight=0.0668, interpMode=RCIM_Cubic, leaveTangentWeight=0.0668, time=0.001, value=0.1765}
          <at 0.1097 0.1765>
          {arriveTangent=0.1845, interpMode=RCIM_Cubic, leaveTangent=0.1845, time=0.1401, value=0.1522}
          {
           arriveTangent=-0.9193
           arriveTangentWeight=0.0324
           interpMode=RCIM_Cubic
           leaveTangent=-0.9193
           leaveTangentWeight=0.0324
           time=0.171
           value=0.2078
          }
          {arriveTangent=0.758, interpMode=RCIM_Cubic, leaveTangent=0.758, time=0.216, value=0.0824}
          {interpMode=RCIM_Cubic, tangentMode=RCTM_User, time=0.29, value=0.298}
          {
           arriveTangent=1.194
           arriveTangentWeight=0.0789
           interpMode=RCIM_Cubic
           leaveTangent=1.194
           leaveTangentWeight=0.0789
           tangentMode=RCTM_User
           tangentWeightMode=RCTWM_WeightedBoth
           time=0.375
           value=0.2876
          }
          {
           arriveTangent=0.0845
           arriveTangentWeight=0.0575
           interpMode=RCIM_Cubic
           leaveTangent=0.0845
           leaveTangentWeight=0.0575
           tangentMode=RCTM_User
           tangentWeightMode=RCTWM_WeightedBoth
           time=0.5
           value=0.5947
          }
          {interpMode=RCIM_Cubic, tangentMode=RCTM_User, time=0.817, value=0.3137}
          {arriveTangent=-1.5352, interpMode=RCIM_Cubic, leaveTangent=-1.5352, time=0.833, value=0.4392}
          {
           arriveTangent=-1.3022
           arriveTangentWeight=0.1327
           interpMode=RCIM_Cubic
           leaveTangent=-2.6654
           leaveTangentWeight=0.1837
           tangentMode=RCTM_Break
           tangentWeightMode=RCTWM_WeightedBoth
           time=0.898
           value=0.1894
          }
          {arriveTangent=-0.128, interpMode=RCIM_Cubic, leaveTangent=-0.128, time=0.9285, value=0.1164}
          {
           arriveTangent=0.6573
           arriveTangentWeight=0.3221
           interpMode=RCIM_Cubic
           leaveTangent=0.6573
           leaveTangentWeight=0.3221
           tangentMode=RCTM_User
           tangentWeightMode=RCTWM_WeightedBoth
           time=1
           value=0.1763
          }
         ]}
        RedCurve {keys [
          {arriveTangent=0.2622, arriveTangentWeight=0.0786, interpMode=RCIM_Cubic, leaveTangentWeight=0.0786, time=0.001, value=0.1098}
          <at 0.1097 0.1098>
          {arriveTangent=1.4072, interpMode=RCIM_Cubic, leaveTangent=1.4072, time=0.1401, value=0.4115}
          {
           arriveTangent=-3.6115
           arriveTangentWeight=0.063
           interpMode=RCIM_Cubic
           leaveTangent=-3.6115
           leaveTangentWeight=0.063
           time=0.171
           value=0.349
          }
          {arriveTangent=-0.4284, interpMode=RCIM_Cubic, leaveTangent=-0.4284, time=0.216, value=0.1373}
          {interpMode=RCIM_Cubic, tangentMode=RCTM_User, time=0.29, value=0.298}
          {
           arriveTangent=1.194
           arriveTangentWeight=0.0789
           interpMode=RCIM_Cubic
           leaveTangent=1.194
           leaveTangentWeight=0.0789
           tangentMode=RCTM_User
           tangentWeightMode=RCTWM_WeightedBoth
           time=0.375
           value=0.254
          }
          {
           arriveTangent=0.0426
           arriveTangentWeight=0.1136
           interpMode=RCIM_Cubic
           leaveTangent=0.0426
           leaveTangentWeight=0.1136
           tangentMode=RCTM_User
           tangentWeightMode=RCTWM_WeightedBoth
           time=0.5
           value=0.4193
          }
          {interpMode=RCIM_Cubic, tangentMode=RCTM_User, time=0.817, value=0.3137}
          {arriveTangent=-0.0828, interpMode=RCIM_Cubic, leaveTangent=-0.0828, tangentMode=RCTM_User, time=0.833, value=0.4392}
          {
           arriveTangent=-0.8044
           arriveTangentWeight=0.2417
           interpMode=RCIM_Cubic
           leaveTangent=-0.8044
           leaveTangentWeight=0.288
           tangentMode=RCTM_User
           time=0.898
           value=0.4271
          }
          {arriveTangent=-3.1243, interpMode=RCIM_Cubic, leaveTangent=-3.1243, time=0.9285, value=0.1053}
          {arriveTangentWeight=0.3333, interpMode=RCIM_Cubic, tangentMode=RCTM_User, tangentWeightMode=RCTWM_WeightedBoth, time=1, value=0.1085}
         ]}
       }}
      CurveIndex=Linked() {linkedVariable {name=NPC.FortniteNPC.TimeOfDay_0-1, type=NiagaraFloat}}
     ))
    ScaleA=true
    ScaleRGB=true
   )
  ]
  rendererClasses [&NiagaraMeshRendererProperties]
  renderers [{rendererClass=&NiagaraMeshRendererProperties, rendererIndex=0}]
 }
}
summary {
 systemName=NS_Emote_AlmondSplash
 emitters [
  {emitterName=SnowDust, rendererClasses [&NiagaraSpriteRendererProperties]}
  {emitterName=SnowFlakes, rendererClasses [&NiagaraSpriteRendererProperties]}
  {bEnabled=false, emitterName=SnowDust001, rendererClasses [&NiagaraSpriteRendererProperties]}
  {bEnabled=false, emitterName=SnowFlakes001, rendererClasses [&NiagaraSpriteRendererProperties]}
  {emitterName=SnowWake, rendererClasses [&NiagaraMeshRendererProperties]}
 ]
 userVariables [
  User.bIsFrontEnd {type=NiagaraBool}
  User.bIsFrontEndPreview {type=NiagaraBool}
 ]
}
